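# Sound cue that plays the collection foley for the duct tape ingredient loot.
# A Mixer layers two wave players, a "Real" and a "Surreal" collect sound, at equal input volume, with the overall volume reduced.
# It uses the shared prop-collect attenuation and loot-collect sound class.
# A concurrency override allows two simultaneous instances and stops the oldest when exceeded.
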SoundCue /Game/Sounds/Foley_Loot/Ingredients/DuctTape/DuctTape_Collect_Cue
root=Mixer_4
nodes [
 {id=Mixer_4, children [WavePlayer_9, WavePlayer_10], properties {inputVolume [1,1]}}
 {id=WavePlayer_10, properties {soundWaveAssetPtr=&/DuctTape_Collect_Surreal}}
 {id=WavePlayer_9, properties {soundWaveAssetPtr=&/DuctTape_Collect_Real}}
]
properties {
 attenuationSettings=/Game/Sounds/Attenuation/Fort_Prop_Collect_Att
 bOverrideConcurrency=true
 maxDistance=2097152
 soundClassObject=&Fort_Loot_Collect
 volumeMultiplier=0.251
 concurrencyOverrides {maxCount=2, resolutionRule=StopOldest}
}
